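# Niagara effect for the burger in the CyclopsPrey emote; despite the directional grease name, only its Flash emitter is enabled.
# Flash bursts one short-lived, faint sprite, tinted by time of day, that grows while its alpha rises and fades.
# The disabled Grease emitter sprays subUV droplets from a ring shape under gravity, with burst count reduced in the front end and one velocity scaled by User.Intensity.
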NiagaraSystem /BRCosmetics/Animation/Game/MainPlayer/Emotes/CyclopsPrey/FX/NS_Emote_CyclopsPrey_Burger_Directional_Grease
dependencies {
 usedDataInterfaces [
  &NiagaraDataInterfaceSpriteRendererInfo
  &NiagaraDataInterfaceCurve
  &NiagaraDataInterfaceArrayFloat
 ]
 usedDynamicInputs [&SelectFloatFromArray]
 usedModules [
  &SystemState
  &self:SystemSpawnScript.SystemScriptSource.SystemScriptGraph.NiagaraNodeAssignment_0.SetVariables_#1
 ]
}
emitters {
 Flash {
  emitterUpdateScript [
   &EmitterState(
    Inactive Response=::NewEnumerator1
    Life Cycle Mode=::NewEnumerator1
    Loop Behavior=::NewEnumerator1
    Loop Delay=0.5
    Loop Duration=0.2
    UseLoopDelay=false
   )
   &SpawnBurst_Instantaneous(Spawn Count=1, Use Spawn Probability=false)
  ]
  particleSpawnScript [
   &InitializeParticle(
    Color=&MultiplyLinearColor(
     Linear Color=&FNTimeOfDayColor(
      Night Color=<rgba 0.1509 0.1982 0.3594 1>
      Morning Color=<rgba 0.4592 0.4876 0.7865 1>
      Day Color=<rgba 1 1 1 1>
      Evening Color=<rgba 0.4592 0.4876 0.7865 1>
     )
     Scale Factor=&MakeVector4_FromLinearColor(LinearColor=<rgba 0.6979 0.5444 0.4725 0.15>)
    )
    Lifetime=0.15
    Sprite Size Mode=::NewEnumerator3
    Uniform Sprite Size=50
    UsePositionOffset=false
    Write Material Random=true
   ) {module=InitializeParticle001}
   &DynamicMaterialParameters(
    Index 0 Param 1=2
    Module.Param0WriteEnabled=true
    Module.Param1WriteEnabled=true
    Module.Param2WriteEnabled=true
    Module.Param3WriteEnabled=true
   )
   &CameraOffset(Camera Offset Amount=-5)
   &CameraOffset(Camera Offset Amount=-1) {enabled=false, module=CameraOffset001}
  ]
  particleUpdateScript [
   &ParticleState()
   &ScaleSpriteSize(Scale Factor=&Vector2DFromFloat(Value=&FloatFromCurve(FloatCurve=DataInterface() {propertyValues {Curve {keys [<at 0 1 RCIM_Cubic>,<at 1 2 RCIM_Cubic>]}}})))
   &ScaleColor(
    Scale Alpha=&FloatFromCurve(FloatCurve=DataInterface() {propertyValues {Curve {keys [<at 0 0>,<at 0.25 0.75>,<at 1 0>]}}})
    ScaleA=true
    ScaleRGB=true
   ) {module=ScaleColor001}
  ]
  rendererClasses [&NiagaraSpriteRendererProperties]
  renderers [{rendererClass=&NiagaraSpriteRendererProperties}]
 }
 Grease {
  bEnabled=false
  emitterUpdateScript [
   &EmitterState(
    Enable Distance Culling=true
    Life Cycle Mode=::NewEnumerator1
    Loop Behavior=::NewEnumerator1
    Loop Duration=0.1
    MaxDistance=9000
    Module.bUseMaxDistance=true
    Module.bUseMinDistance=false
    Scalability Mode=::NewEnumerator1
    UseLoopDelay=false
   )
   &SpawnRate(SpawnRate=40, Use Spawn Probability=false)
   &SpawnBurst_Instantaneous(
    Spawn Count=&MakeIntFromFloat(Float=&FN_FrontEndFloatModifier(Gameplay=3, Front End=0))
    Use Spawn Probability=false
    UseLoopCountLimit=false
   )
  ]
  particleSpawnScript [
   &InitializeParticle(
    Color=&MultiplyLinearColor(
     Linear Color=&FNTimeOfDayColor(
      Night Color=<rgba 0.7865 0.7865 0.7865 0.3>
      Morning Color=<rgba 0.7865 0.7865 0.7865 0.5>
      Day Color=<rgba 1 1 1 1>
      Evening Color=<rgba 0.7865 0.7865 0.7865 0.5>
     )
     Scale Factor=&MakeVector4_FromLinearColor(LinearColor=<rgba 0.1771 0.0864 0.0498 0.8>)
    )
    Lifetime=0.5
    Sprite Rotation Mode=::NewEnumerator2
    Sprite Size Max=<1.5 6>
    Sprite Size Min=<1.5 5>
    Sprite Size Mode=::NewEnumerator2
    Sprite UV Mode=::NewEnumerator3
    UsePositionOffset=false
    Write Material Random=true
   )
   &SubUVAnimation(
    "Sprite Renderer "=DataInterface() {propertyValues {SpriteRenderer=&self:Grease_0.NiagaraSpriteRendererProperties_0}}
    SubUV Animation Mode:ENiagara_SubUVLookupModeV2=::NewEnumerator2
    UseEndFrame=false
    UseStartFrame=false
   )
   &ShapeLocation(
    MaskSpawnGroup=false
    Override Seed=false
    Ring / Disc Distribution Mode:ENiagaraSphereDistributionMode=::NewEnumerator1
    Ring Radius=15
    Rotation Angle Type:ENiagara_AngleInput=::NewEnumerator0
    Rotation Axis=<0 0 0>
    Rotation Coordinate Space=::Simulation
    Rotation Mode=::NewEnumerator1
    Shape Primitive=::NewEnumerator4
    U Position=&RandomRangeFloat(Minimum=-0.15, Maximum=0.15)
    UseOwnerScale=true
   )
   &AddVelocity(
    Cone Angle=10
    Cone Axis=<0 0 1>
    Module.UseConeFalloff=false
    Module.UseVelDistribution=false
    Override Seed=false
    Velocity Mode=::NewEnumerator2
    Velocity Speed=&RandomRangeFloat(Minimum=60, Maximum=150)
   ) {module=AddVelocity001}
   &AddVelocity(
    Module.Use Falloff Distance=false
    Override Seed=false
    Velocity Mode=::NewEnumerator1
    Velocity Speed=&Multiply_Float(A=&RandomRangeFloat(Minimum=300, Maximum=150), B=$User.Intensity)
   )
   &CameraOffset(Camera Offset Amount=&RandomRangeFloat(Minimum=10, Maximum=15))
  ]
  particleUpdateScript [
   &ParticleState()
   &GravityForce(Gravity=<0 0 -1000>)
   &Drag(Drag=0, Use Linear Drag=true, Use Rotational Drag=false)
   &SolveForcesAndVelocity(Clamp Velocity=false, Limit Acceleration=false)
   &ScaleSpriteSize(Uniform Curve Sprite Scale=DataInterface() {propertyValues {Curve {keys [
        {arriveTangent=4.1785, interpMode=RCIM_Cubic, leaveTangent=4.1785, tangentMode=RCTM_User, time=0, value=0.15}
        {arriveTangent=1.1179, interpMode=RCIM_Cubic, leaveTangent=1.1179, tangentMode=RCTM_User, time=0.2042, value=0.6228}
        {arriveTangent=0.0224, interpMode=RCIM_Cubic, leaveTangent=0.0224, tangentMode=RCTM_User, time=1, value=1}
       ]}}})
   &ScaleSpriteSize(
    Scale Factor=&Lerp_Vector2(A=<1 1>, B=<1 0.5>, Alpha=$Particles.NormalizedAge)
    Scale Sprite Size Mode=::NewEnumerator2
   ) {module=ScaleSpriteSize001}
   &ScaleColor(
    Scale Alpha=&FloatFromCurve(
     FloatCurve=DataInterface() {propertyValues {Curve {defaultValue=0, keys [{tangentMode=RCTM_User, time=0, value=1}, {tangentMode=RCTM_User, time=0.6, value=0}]}}}
     Scale Curve=5
    )
    ScaleA=true
    ScaleRGB=true
   ) {module=ScaleColor002}
   &KillParticlesInVolume(Box Size:Vector3f=<50 50 50>, Invert Volume=true, Kill Shape=::NewEnumerator1)
  ]
  rendererClasses [&NiagaraSpriteRendererProperties]
  renderers [{rendererClass=&NiagaraSpriteRendererProperties}]
 }
}
summary {
 systemName=NS_Emote_CyclopsPrey_Burger_Directional_Grease
 emitters [
  {bEnabled=false, emitterName=Grease, rendererClasses [&NiagaraSpriteRendererProperties]}
  {emitterName=Flash, rendererClasses [&NiagaraSpriteRendererProperties]}
 ]
 userVariables [
  User.bIsFrontEnd {type=NiagaraBool}
  User.bIsFrontEndPreview {type=NiagaraBool}
  User.Intensity {type=NiagaraFloat, defaultValue=0.4, description=""}
  User.NormalizedNotifyProgress {type=NiagaraFloat, defaultValue=0, description=""}
 ]
}
