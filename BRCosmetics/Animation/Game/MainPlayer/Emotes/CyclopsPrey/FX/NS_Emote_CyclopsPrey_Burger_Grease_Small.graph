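# Niagara system for the small burger grease effect in the CyclopsPrey player emote, built from a single Flash sprite emitter.
# A one-particle burst spawns a very short-lived, faint sprite, tinted by a time-of-day color blend and shifted along the camera axis.
# Over its life the sprite grows while curve-driven alpha quickly fades in and out.
# Dynamic material parameters and a material random value feed the sprite material.
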NiagaraSystem /BRCosmetics/Animation/Game/MainPlayer/Emotes/CyclopsPrey/FX/NS_Emote_CyclopsPrey_Burger_Grease_Small
dependencies {
 usedDataInterfaces [&NiagaraDataInterfaceCurve, &NiagaraDataInterfaceArrayFloat]
 usedDynamicInputs [&SelectFloatFromArray]
 usedModules [
  &SystemState
  &self:SystemSpawnScript.SystemScriptSource.SystemScriptGraph.NiagaraNodeAssignment_0.SetVariables_#1
 ]
}
emitters {Flash {
  emitterUpdateScript [
   &EmitterState(
    Inactive Response=::NewEnumerator1
    Life Cycle Mode=::NewEnumerator1
    Loop Behavior=::NewEnumerator1
    Loop Delay=0.5
    Loop Duration=0.2
    UseLoopDelay=false
   )
   &SpawnBurst_Instantaneous(Spawn Count=1, Use Spawn Probability=false)
  ]
  particleSpawnScript [
   &InitializeParticle(
    Color=&MultiplyLinearColor(
     Linear Color=&FNTimeOfDayColor(
      Night Color=<rgba 0.1509 0.1982 0.3594 1>
      Morning Color=<rgba 0.4592 0.4876 0.7865 1>
      Day Color=<rgba 1 1 1 1>
      Evening Color=<rgba 0.4592 0.4876 0.7865 1>
     )
     Scale Factor=&MakeVector4_FromLinearColor(LinearColor=<rgba 0.6979 0.5444 0.4725 0.15>)
    )
    Lifetime=0.15
    Sprite Size Mode=::NewEnumerator3
    Uniform Sprite Size=40
    UsePositionOffset=false
    Write Material Random=true
   ) {module=InitializeParticle001}
   &DynamicMaterialParameters(
    Index 0 Param 1=2
    Module.Param0WriteEnabled=true
    Module.Param1WriteEnabled=true
    Module.Param2WriteEnabled=true
    Module.Param3WriteEnabled=true
   )
   &CameraOffset(Camera Offset Amount=-5)
   &CameraOffset(Camera Offset Amount=-1) {enabled=false, module=CameraOffset001}
  ]
  particleUpdateScript [
   &ParticleState()
   &ScaleSpriteSize(Scale Factor=&Vector2DFromFloat(Value=&FloatFromCurve(FloatCurve=DataInterface() {propertyValues {Curve {keys [<at 0 1 RCIM_Cubic>,<at 1 2 RCIM_Cubic>]}}})))
   &ScaleColor(
    Scale Alpha=&FloatFromCurve(FloatCurve=DataInterface() {propertyValues {Curve {keys [<at 0 0>,<at 0.25 0.75>,<at 1 0>]}}})
    ScaleA=true
    ScaleRGB=true
   ) {module=ScaleColor001}
  ]
  rendererClasses [&NiagaraSpriteRendererProperties]
  renderers [{rendererClass=&NiagaraSpriteRendererProperties}]
 }}
summary {
 systemName=NS_Emote_CyclopsPrey_Burger_Grease_Small
 emitters [{emitterName=Flash, rendererClasses [&NiagaraSpriteRendererProperties]}]
 userVariables [
  User.bIsFrontEnd {type=NiagaraBool}
  User.bIsFrontEndPreview {type=NiagaraBool}
  User.Intensity {type=NiagaraFloat, defaultValue=0.4, description=""}
  User.NormalizedNotifyProgress {type=NiagaraFloat, defaultValue=0, description=""}
 ]
}
