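# Niagara ribbon trail effect for the burger patty in the CyclopsPrey emote.
# A Position emitter keeps burst-spawned particles on the gadget skeletal mesh's patty bone with a slight random offset.
# The Trail emitter samples those particles to lay short-lived ribbons, whose spawn width and alpha follow curves over the emitter loop.
# Ribbons are tinted by time of day, then narrow and fade as they age.
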NiagaraSystem /BRCosmetics/Animation/Game/MainPlayer/Emotes/CyclopsPrey/FX/NS_Emote_CyclopsPrey_Burger_Trail
dependencies {
 usedDataInterfaces [
  &NiagaraDataInterfaceParticleRead
  &NiagaraDataInterfaceColorCurve
  &NiagaraDataInterfaceCurve
  &NiagaraDataInterfaceSkeletalMesh
  &NiagaraDataInterfaceArrayFloat
 ]
 usedDynamicInputs [&SelectFloatFromArray]
 usedModules [
  &SystemState
  &self:SystemSpawnScript.SystemScriptSource.SystemScriptGraph.NiagaraNodeAssignment_0.SetVariables_#1
 ]
}
emitters {
 Position {
  emitterUpdateScript [
   &EmitterState(Life Cycle Mode=::NewEnumerator1, Loop Behavior=::NewEnumerator1, Loop Duration=1, UseLoopDelay=false)
   &SpawnBurst_Instantaneous(Use Spawn Probability=false, UseLoopCountLimit=false)
  ]
  particleSpawnScript [&InitializeParticle(Color Mode=::NewEnumerator0, Lifetime=2, UsePositionOffset=false, Write Material Random=true)]
  particleUpdateScript [
   &ParticleState()
   &SkeletalMeshLocation(
    Bone Sampling Mode=::NewEnumerator4
    MaskSpawnGroup=false
    Orientation Sampling=::NewEnumerator1
    Sampled Position Offset=&RandomRangeVector(Minimum=<-1 0 0>, Maximum=<1 1 1>)
    Skeletal Mesh=DataInterface() {propertyValues {
      DefaultMesh=/BRCosmetics/Animation/Props/Gadgets_CyclopsPrey/Meshes/CyclopsPrey_Gadget
      ExcludeBoneName=root
      OptionalComponentClass=null
      PreviewMesh=&CyclopsPrey_Gadget
      SkinningMode=SkinOnTheFly
      SoftSourceActor=null
      SourceMode=Default
      UvSetIndex=0
      WholeMeshLOD=-1
      bExcludeBone=false
      bReadDeformedGeometry=true
      bRequireCurrentFrameData=true
      ComponentTags []
      FilteredBones [JNT_BurgerPatty_A]
      FilteredSockets []
      MeshUserParameter {parameter {name=None, typeDefHandle {registeredTypeIndex=1}, varData []}}
      SamplingRegions []
     }}
   ) {module=SkeletalMeshLocation001}
  ]
  rendererClasses [&NiagaraSpriteRendererProperties]
  renderers [{rendererClass=&NiagaraSpriteRendererProperties}]
 }
 Trail {
  emitterUpdateScript [
   &EmitterState(Life Cycle Mode=::NewEnumerator1, Loop Behavior=::NewEnumerator1, Loop Duration=0.6, UseLoopDelay=false)
   &SpawnRate(SpawnRate=30, Use Spawn Probability=false)
   &SpawnParticlesFromOtherEmitter(
    Attribute Reader=DataInterface() {propertyValues {EmitterBinding {bindingMode=Other, emitterName=Position}}}
    Spawn Rate=0
    Use Maximum Age=false
    Use Minimum Age=false
   )
  ]
  particleSpawnScript [
   &InitializeParticle(
    Color=&MultiplyLinearColor(
     Linear Color=&FNTimeOfDayColor(
      Night Color=<rgba 0.0867 0.109 0.2969 1>
      Morning Color=<rgba 0.3094 0.3911 0.5885 1>
      Day Color=<rgba 1 1 1 1>
      Evening Color=<rgba 0.3094 0.3911 0.5885 1>
     )
     Scale Factor=&MakeVector4_FromLinearColor(LinearColor=&ColorFromCurve(
       ColorCurve=DataInterface() {propertyValues {
         AlphaCurve {keys [<at 0 0 RCIM_Cubic>,<at 0.065 0.15>,<at 1 0 RCIM_Cubic>]}
         BlueCurve {keys [<at 0 1 RCIM_Cubic>,<at 1 1 RCIM_Cubic>]}
         GreenCurve {keys [<at 0 1 RCIM_Cubic>,<at 1 1 RCIM_Cubic>]}
         RedCurve {keys [<at 0 1 RCIM_Cubic>,<at 1 1 RCIM_Cubic>]}
        }}
       CurveIndex=$Emitter.NormalizedLoopAge
      ))
    )
    Lifetime=0.45
    Ribbon Width=&FloatFromCurve(FloatCurve=DataInterface() {propertyValues {Curve {keys [<at 0 0>,<at 1 0.95>]}}}, CurveIndex=$Emitter.NormalizedLoopAge, Scale Curve=20)
    Ribbon Width Mode=::NewEnumerator1
    UsePositionOffset=false
    Write Material Random=true
   )
   &SampleParticlesFromOtherEmitter()
  ]
  particleUpdateScript [
   &ParticleState()
   &SolveForcesAndVelocity(Clamp Velocity=false, Limit Acceleration=false)
   &ScaleRibbonWidth(Ribbon Width Scale=&FloatFromCurve(FloatCurve=DataInterface() {propertyValues {Curve {keys [<at 0 1>,<at 1 0>]}}}))
   &ScaleColor(
    Scale Alpha=&FloatFromCurve(FloatCurve=DataInterface() {propertyValues {Curve {keys [<at 0 1>,<at 1 0>]}}})
    ScaleA=true
    ScaleRGB=true
   )
  ]
  rendererClasses [&NiagaraRibbonRendererProperties]
  renderers [{rendererClass=&NiagaraRibbonRendererProperties}]
 }
}
summary {
 systemName=NS_Emote_CyclopsPrey_Burger_Trail
 emitters [
  {emitterName=Trail, rendererClasses [&NiagaraRibbonRendererProperties]}
  {emitterName=Position, rendererClasses [&NiagaraSpriteRendererProperties]}
 ]
 userVariables [
  User.bIsFrontEnd {type=NiagaraBool}
  User.bIsFrontEndPreview {type=NiagaraBool}
  User.FX_WeaponCaseSize {type=NiagaraInt32, defaultValue=1}
  User.NormalizedNotifyProgress {type=NiagaraFloat, defaultValue=0, description=""}
  User.SystemScale {type=NiagaraFloat, defaultValue=1, description=""}
 ]
}
