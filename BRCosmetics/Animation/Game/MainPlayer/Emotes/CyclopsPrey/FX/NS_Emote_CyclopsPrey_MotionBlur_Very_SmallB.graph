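# Niagara system for the CyclopsPrey emote that draws small motion-blur streaks on the gadget prop's JNT_BurgerPatty_B bone.
# Two looping mesh emitters, Base_Ring and thinner Highlights, spawn short-lived rotating meshes that follow the bone through per-update skeletal mesh sampling.
# Curves drive mesh growth, alpha fade and a dynamic material parameter.
# Colour is randomised, tinted by time of day and dimmed in the front end.
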NiagaraSystem /BRCosmetics/Animation/Game/MainPlayer/Emotes/CyclopsPrey/FX/NS_Emote_CyclopsPrey_MotionBlur_Very_SmallB
dependencies {
 usedDataInterfaces [&NiagaraDataInterfaceCurve, &NiagaraDataInterfaceSkeletalMesh]
 usedDynamicInputs [&MakeVector, &RandomRangeFloat]
 usedModules [&SystemState]
}
emitters {
 Base_Ring {
  emitterUpdateScript [
   &EmitterState(Life Cycle Mode=::NewEnumerator1, Loop Behavior=::NewEnumerator1, Loop Duration=10, UseLoopDelay=false)
   &SpawnRate(SpawnRate=25, Use Spawn Probability=false)
  ]
  particleSpawnScript [
   &PartitionParticles()
   &InitializeParticle(
    Color=&MultiplyLinearColor(
     Linear Color=&FNTimeOfDayColor(
      Night Color=<rgba 0.5716 0.5664 0.9375 1>
      Morning Color=<rgba 0.8073 0.6385 0.4858 1>
      Day Color=<rgba 1 0.8066 0.6198 1>
      Evening Color=<rgba 0.8073 0.6385 0.4858 1>
     )
     Scale Factor=/Niagara/DynamicInputs/Multiply/Multiply_Vector4_ByFloat(
      Vector4:Vector4f=&MakeVector4_FromLinearColor(LinearColor=&RandomRangeLinearColor(Minimum=<rgba 0.3021 0.2532 0.2297 0.75>, Maximum=<rgba 0.9219 0.7159 0.6098 1>, Color Channel Mode=::NewEnumerator0))
      Float:NiagaraFloat=&FN_FrontEndFloatModifier(Gameplay=1, Front End=0.25)
     )
    )
    Lifetime=0.25
    Mesh Scale Max=&Multiply_VectorByFloat(Vector=<1 1.5 1>, Float=0.07)
    Mesh Scale Min=&Multiply_VectorByFloat(Vector=<1 1.5 1>, Float=0.05)
    Mesh Scale Mode=::NewEnumerator2
    Position Offset=&RandomRangeVector(Minimum=<-1 -1 -2>, Maximum=<1 1 2>)
    UsePositionOffset=true
    Write Material Random=true
   )
   &CameraOffset(Camera Offset Amount=15)
   &InitialMeshOrientation(
    Mesh Axis To Orient=<0 0 0>
    Mesh Orientation Mode=::NewEnumerator2
    Orientation Coordinate Space=::Local
    Orientation Vector=<0 1 0>
    Rotation=&RandomRangeVector(Minimum=<0 0 0>, Maximum=<1 0 0>)
    Rotation Coordinate Space=::NewEnumerator2
    Use Rotation Vector=true
   )
  ]
  particleUpdateScript [
   &ParticleState()
   &ScaleMeshSize(Scale Factor=<1 1.5 1>) {module=ScaleMeshSize002}
   &ApplyOwnerScaleToAttributes(Scale Initial Mesh Scale=true)
   &UpdateMeshOrientation(Rotation Rate=0.5, Rotation Vector=<0 -3 0>) {module=UpdateMeshOrientation002}
   &ScaleMeshSize(Scale Factor=&VectorFromFloat(Value=&FloatFromCurve(FloatCurve=DataInterface() {propertyValues {Curve {keys [<at 0 0.7>,<at 0.45 0.875>,<at 1 1.5>]}}})))
   &DynamicMaterialParameters(
    Index 0 Param 1=&FloatFromCurve(FloatCurve=DataInterface() {propertyValues {Curve {keys [
         {arriveTangent=2.4535, interpMode=RCIM_Cubic, leaveTangent=2.4535, tangentMode=RCTM_User, time=0, value=0}
         {interpMode=RCIM_Cubic, tangentMode=RCTM_User, time=0.5, value=0.5}
         {arriveTangent=-2.5865, tangentMode=RCTM_User, time=1, value=0}
        ]}}})
    Module.Index0Param0Write=true
    Module.Index0Param1Write=false
    Module.Index0Param2Write=false
    Module.Index0Param3Write=false
   )
   &SolveForcesAndVelocity(Clamp Velocity=false, Limit Acceleration=false)
   &ScaleColor(
    Scale Alpha=&FloatFromCurve(
     FloatCurve=DataInterface() {propertyValues {Curve {keys [
         {arriveTangent=6.3511, interpMode=RCIM_Cubic, leaveTangent=6.3511, tangentMode=RCTM_User, time=0, value=0}
         {interpMode=RCIM_Cubic, tangentMode=RCTM_User, time=0.5, value=1}
         {tangentMode=RCTM_User, time=1, value=0}
        ]}}}
     Scale Curve=0.6
    )
    Scale RGB=&MakeVectorFromLinearColorRGB(Color=&FNTimeOfDayColor(
      Night Color=<rgba 0.1253 0.1914 0.5 1>
      Morning Color=<rgba 0.5 0.3383 0.3719 1>
      Day Color=<rgba 0.5 0.5 0.5 1>
      Evening Color=<rgba 0.4474 0.4547 0.4578 1>
     ))
    ScaleA=true
    ScaleRGB=true
   )
   &ScaleColor(Scale Alpha=0.25, ScaleA=true, ScaleRGB=true) {module=ScaleColor001}
   &SkeletalMeshLocation(
    Bone Sampling Mode=::NewEnumerator4
    MaskSpawnGroup=false
    Orientation Sampling=::NewEnumerator1
    Sampled Position Offset=&RandomRangeVector(Minimum=<-1 0 0>, Maximum=<1 1 1>)
    Skeletal Mesh=DataInterface() {propertyValues {
      DefaultMesh=/BRCosmetics/Animation/Props/Gadgets_CyclopsPrey/Meshes/CyclopsPrey_Gadget
      ExcludeBoneName=root
      OptionalComponentClass=null
      PreviewMesh=&CyclopsPrey_Gadget
      SkinningMode=SkinOnTheFly
      SoftSourceActor=null
      SourceMode=Default
      UvSetIndex=0
      WholeMeshLOD=-1
      bExcludeBone=false
      bReadDeformedGeometry=true
      bRequireCurrentFrameData=true
      ComponentTags []
      FilteredBones [JNT_BurgerPatty_B]
      FilteredSockets []
      MeshUserParameter {parameter {name=None, typeDefHandle {registeredTypeIndex=1}, varData []}}
      SamplingRegions []
     }}
   ) {module=SkeletalMeshLocation001}
  ]
  rendererClasses [&NiagaraMeshRendererProperties]
  renderers [{rendererClass=&NiagaraMeshRendererProperties, rendererIndex=0}]
 }
 Highlights {
  emitterUpdateScript [
   &EmitterState(Life Cycle Mode=::NewEnumerator1, Loop Behavior=::NewEnumerator1, Loop Duration=10, UseLoopDelay=false)
   &SpawnRate(SpawnRate=25, Use Spawn Probability=false)
  ]
  particleSpawnScript [
   &PartitionParticles()
   &InitializeParticle(
    Color=&MultiplyLinearColor(
     Linear Color=&FNTimeOfDayColor(
      Night Color=<rgba 0.5716 0.5664 0.9375 1>
      Morning Color=<rgba 0.8073 0.6385 0.4858 1>
      Day Color=<rgba 1 0.8066 0.6198 1>
      Evening Color=<rgba 0.8073 0.6385 0.4858 1>
     )
     Scale Factor=&Multiply_Vector4_ByFloat(
      Vector4:Vector4f=&MakeVector4_FromLinearColor(LinearColor=&RandomRangeLinearColor(Minimum=<rgba 0.3021 0.2532 0.2297 0.75>, Maximum=<rgba 0.9219 0.7159 0.6098 1>, Color Channel Mode=::NewEnumerator0))
      Float:NiagaraFloat=&FN_FrontEndFloatModifier(Gameplay=1, Front End=0.25)
     )
    )
    Lifetime=0.25
    Mesh Scale Max=&Multiply_VectorByFloat(Vector=<1 0.3 1>, Float=0.07)
    Mesh Scale Min=&Multiply_VectorByFloat(Vector=<1 0.1 1>, Float=0.07)
    Mesh Scale Mode=::NewEnumerator2
    Position Offset=&RandomRangeVector(Minimum=<-7 0 0>, Maximum=<7 0 0>)
    Position Offset Coordinate Space=::Local
    UsePositionOffset=true
    Write Material Random=true
   )
   &CameraOffset(Camera Offset Amount=15)
   &InitialMeshOrientation(
    Mesh Axis To Orient=<0 0 0>
    Mesh Orientation Mode=::NewEnumerator2
    Orientation Coordinate Space=::Local
    Orientation Vector=<0 1 0>
    Rotation=&RandomRangeVector(Minimum=<0 0 0>, Maximum=<1 0 0>)
    Rotation Coordinate Space=::NewEnumerator2
    Use Rotation Vector=true
   )
  ]
  particleUpdateScript [
   &ParticleState()
   &ScaleMeshSize(Scale Factor=<1 1.5 1>) {module=ScaleMeshSize002}
   &ApplyOwnerScaleToAttributes(Scale Initial Mesh Scale=true)
   &UpdateMeshOrientation(Rotation Rate=0.5, Rotation Vector=<0 -3 0>) {module=UpdateMeshOrientation002}
   &ScaleMeshSize(Scale Factor=&VectorFromFloat(Value=&FloatFromCurve(FloatCurve=DataInterface() {propertyValues {Curve {keys [<at 0 0.7>,<at 0.45 0.875>,<at 1 1.5>]}}}, Scale Curve=0.9)))
   &DynamicMaterialParameters(
    Index 0 Param 1=&FloatFromCurve(FloatCurve=DataInterface() {propertyValues {Curve {keys [
         {arriveTangent=2.4535, interpMode=RCIM_Cubic, leaveTangent=2.4535, tangentMode=RCTM_User, time=0, value=0}
         {interpMode=RCIM_Cubic, tangentMode=RCTM_User, time=0.5, value=0.5}
         {arriveTangent=-2.5865, tangentMode=RCTM_User, time=1, value=0}
        ]}}})
    Module.Index0Param0Write=true
    Module.Index0Param1Write=false
    Module.Index0Param2Write=false
    Module.Index0Param3Write=false
   )
   &SolveForcesAndVelocity(Clamp Velocity=false, Limit Acceleration=false)
   &ScaleColor(
    Scale Alpha=&FloatFromCurve(
     FloatCurve=DataInterface() {propertyValues {Curve {keys [
         {arriveTangent=6.3511, interpMode=RCIM_Cubic, leaveTangent=6.3511, tangentMode=RCTM_User, time=0, value=0}
         {interpMode=RCIM_Cubic, tangentMode=RCTM_User, time=0.5, value=1}
         {tangentMode=RCTM_User, time=1, value=0}
        ]}}}
     Scale Curve=0.6
    )
    Scale RGB=&MakeVectorFromLinearColorRGB(Color=&FNTimeOfDayColor(
      Night Color=<rgba 0.1253 0.1914 0.5 1>
      Morning Color=<rgba 0.5 0.3383 0.3719 1>
      Day Color=<rgba 0.5 0.5 0.5 1>
      Evening Color=<rgba 0.4474 0.4547 0.4578 1>
     ))
    ScaleA=true
    ScaleRGB=true
   )
   &ScaleColor(ScaleA=true, ScaleRGB=true) {module=ScaleColor001}
   &SkeletalMeshLocation(
    Bone Sampling Mode=::NewEnumerator4
    MaskSpawnGroup=false
    Orientation Sampling=::NewEnumerator1
    Sampled Position Offset=&RandomRangeVector(Minimum=<5 0 0>, Maximum=<-5 1 1>)
    Skeletal Mesh=DataInterface() {propertyValues {
      DefaultMesh=&CyclopsPrey_Gadget
      ExcludeBoneName=root
      OptionalComponentClass=null
      PreviewMesh=&CyclopsPrey_Gadget
      SkinningMode=SkinOnTheFly
      SoftSourceActor=null
      SourceMode=Default
      UvSetIndex=0
      WholeMeshLOD=-1
      bExcludeBone=false
      bReadDeformedGeometry=true
      bRequireCurrentFrameData=true
      ComponentTags []
      FilteredBones [JNT_BurgerPatty_B]
      FilteredSockets []
      MeshUserParameter {parameter {name=None, typeDefHandle {registeredTypeIndex=1}, varData []}}
      SamplingRegions []
     }}
   )
  ]
  rendererClasses [&NiagaraMeshRendererProperties]
  renderers [{rendererClass=&NiagaraMeshRendererProperties, rendererIndex=0}]
 }
}
summary {
 systemName=NS_Emote_CyclopsPrey_MotionBlur_Very_SmallB
 emitters [
  {emitterName=Base_Ring, rendererClasses [&NiagaraMeshRendererProperties]}
  {emitterName=Highlights, rendererClasses [&NiagaraMeshRendererProperties]}
 ]
 userVariables [
  User.bIsFrontEnd {type=NiagaraBool}
  User.bIsFrontEndPreview {type=NiagaraBool}
 ]
}
