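# Niagara system for the CyclopsPrey emote that adds sizzle smoke and steam at three burger positions.
# Ambient_Steam emits soft SubUV sprites, while Burger_A, B and C emit rising, tapering ribbons with optional emissive color and time-of-day opacity.
# User offsets lerp spawn points between preset positions, and separate alphas fade A and B.
# The global opacity multiplier modules are disabled, so that parameter has no effect.
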NiagaraSystem /BRCosmetics/Animation/Game/MainPlayer/Emotes/CyclopsPrey/FX/NS_Emote_CyclopsPrey_Sizzle_Consolidation
dependencies {
 usedDataInterfaces [&NiagaraDataInterfaceArrayFloat3, &NiagaraDataInterfaceCurve]
 usedDynamicInputs []
 usedModules [&SystemState]
}
emitters {
 Ambient_Steam {
  emitterUpdateScript [
   &EmitterState(Life Cycle Mode=::NewEnumerator1, Loop Duration=1, UseLoopDelay=true)
   &SpawnRate(SpawnRate=3, Use Spawn Probability=false)
  ]
  particleSpawnScript [
   &InitializeParticle(
    Color=<rgba 0.6875 0.6875 0.6875 0.05>
    Lifetime Max=3
    Lifetime Min=2.3
    Lifetime Mode=::NewEnumerator1
    Position Offset=<0 0 4>
    Sprite Rotation Angle=&RandomRangeFloat(Minimum=-0.05, Maximum=0.3)
    Sprite Rotation Mode=::NewEnumerator1
    Sprite Size Mode=::NewEnumerator1
    Uniform Sprite Size Max=30
    Uniform Sprite Size Min=40
    UsePositionOffset=true
    Write Material Random=true
   )
   &SystemLocation(
    MaskSpawnGroup=false
    Offset=&Lerp_Vector(
     A=<0 0 0>
     B=&Subtract_Vector(A=<0.0003 1.9429 -9.3009>, B=<0.0003 -6.7513 -9.5841>)
     Alpha=$User.Smoke_Position_Offset
    )
   ) {enabled=false}
   &SystemLocation(
    MaskSpawnGroup=false
    Offset=&Lerp_Vector(
     A=&SelectVectorFromArray(
      Array Sampling Mode=::NewEnumerator0
      Direct Array Index=?
      Vector Selection Array=DataInterface() {propertyValues {FloatData [<0.0604 -6.6937 -12.725>, <-15.7355 -6.6264 -12.7505>, <17.5817 0.666 -6.6419>]}}
      Random Seed=0
     )
     B=&SelectVectorFromArray(
      Array Sampling Mode=::NewEnumerator0
      Direct Array Index=?
      Vector Selection Array=DataInterface() {propertyValues {FloatData [<0.1129 1.9835 -12.6499>, <-15.7355 -6.6264 -12.7505>, <17.5817 0.666 -6.6419>]}}
      Random Seed=0
     )
     Alpha=$User.Smoke_Position_Offset
    )
   ) {module=SystemLocation001}
   &ShapeLocation(MaskSpawnGroup=false, Override Seed=false, Sphere Radius=10, UseOwnerScale=true)
   &AddVelocity(Velocity=&RandomRangeVector(Minimum=<-2 -2 6>, Maximum=<2 2 10>, Evaluation Type=::NewEnumerator1))
   &InheritVelocity(Module.Limit Speed=true, Velocity Limit=5, Velocity Scale=<0.25 0.25 0.25>)
  ]
  particleUpdateScript [
   &ParticleState()
   &SubUVAnimation(
    Sub UV Play Rate:NiagaraFloat=&RandomRangeFloat(Minimum=0.03, Maximum=0.05)
    SubUV Animation Mode:ENiagara_SubUVLookupMode=::NewEnumerator3
    SubUV Start Time Offset:NiagaraFloat=&RandomRangeFloat(Minimum=0, Maximum=63)
   )
   &DynamicMaterialParameters(
    Index 0 Param 1=&FloatFromCurve(FloatCurve=DataInterface() {propertyValues {Curve {keys [<at 0 -0.25>,<at 1 -1>]}}})
    Index 0 Param 4=&OneMinusFloat(Float=$Particles.NormalizedAge)
    Module.Param0WriteEnabled=true
    Module.Param1WriteEnabled=false
    Module.Param2WriteEnabled=false
    Module.Param3WriteEnabled=true
   )
   &ScaleSpriteSize(Scale Factor=&Vector2DFromFloat(Value=&FloatFromCurve(
      FloatCurve=DataInterface() {propertyValues {Curve {keys [
          {arriveTangent=1.8751, interpMode=RCIM_Cubic, leaveTangent=1.8751, tangentMode=RCTM_User, time=0, value=0.25}
          {arriveTangent=0.72, interpMode=RCIM_Cubic, leaveTangent=0.72, tangentMode=RCTM_User, time=1, value=1.25}
         ]}}}
      Scale Curve=1.5
     )))
   &Drag(Drag=0.5, Use Linear Drag=true, Use Rotational Drag=false)
   &InheritVelocity(Module.Limit Speed=true, Velocity Limit=0.3, Velocity Scale=&VectorFromFloat(Value=0.01)) {module=InheritVelocity001}
   &InheritSourceMovement(ApplyForce=false, ApplyPosition=true, ApplyPositionFalloff=false)
   &SolveForcesAndVelocity(Clamp Velocity=false, Limit Acceleration=false)
   &ScaleColor(
    Scale Alpha=&FloatFromCurve(FloatCurve=DataInterface() {propertyValues {Curve {keys [<at 0 1>,<at 0.3 1>,<at 0.85 1>,<at 1 0>]}}})
    ScaleA=true
    ScaleRGB=true
   )
   &ScaleColor(ScaleA=true, ScaleRGB=true) {module=ScaleColor001}
  ]
  rendererClasses [&NiagaraSpriteRendererProperties]
  renderers [{rendererClass=&NiagaraSpriteRendererProperties}]
 }
 Burger_A {
  emitterSpawnScript [/SetVariables(Emitter.SystemRand01:NiagaraFloat=&RandomRangeFloat(Minimum=0, Maximum=1)) {isSetParameters=true, module=SetVariables_#1}]
  emitterUpdateScript [
   &EmitterState(Life Cycle Mode=::NewEnumerator1, Loop Duration=1, UseLoopDelay=false)
   &SpawnRate(SpawnRate=30, Use Spawn Probability=false)
  ]
  particleSpawnScript [
   &InitializeParticle(
    Color=$User.BaseColorAlpha
    Lifetime=0.8
    Ribbon Facing Mode=::NewEnumerator1
    Ribbon Facing Vector:Vector3f=<1 0 0>
    Ribbon Width=&Multiply_Float(A=15, B=$User.RibbonScale)
    Ribbon Width Mode=::NewEnumerator1
    UsePositionOffset=false
    Write Material Random=true
   )
   &SystemLocation(
    MaskSpawnGroup=false
    Offset=&Lerp_Vector(A=<0.0433 -6.7158 -12.8209>, B=<0.1129 1.9835 -12.6499>, Alpha=$User.Smoke_Position_Offset)
   )
   &SystemLocation(MaskSpawnGroup=false, Offset=&AddVectorToPosition(Position=$Particles.Position, Vector=<0 0 -10>)) {module=SystemLocation001}
   &AddVelocity(Velocity=<0 0 60>)
  ]
  particleUpdateScript [
   &ParticleState()
   &ScaleColor(
    Scale Alpha=&Multiply_Float(
     A=&FloatFromCurve(
      FloatCurve=DataInterface() {propertyValues {Curve {keys [<at 0 0>, {arriveTangent=-1.7714, interpMode=RCIM_Cubic, leaveTangent=-1.7714, tangentMode=RCTM_User, time=0.5, value=1}, <at 0.8 0>]}}}
      Scale Curve=&FNTimeOfDayScalar(Night Scale=1, Morning Scale=0.3, Evening Scale=0.5)
     )
     B=&FloatFromCurve(FloatCurve=DataInterface() {propertyValues {Curve {keys [<at 0 1>,<at 0.3 1>,<at 0.85 1>,<at 1 0>]}}})
    )
    Scale RGB=&MakeVectorFromLinearColorRGB(Color=&MakeCustomLinearColorFromBool(
      Bool=$User.UseEmissiveSmoke
      True Color=&LerpLinearColors(
       StartColor=$User.EmissiveColor
       EndColor=&MultiplyLinearColor(Linear Color=$User.EmissiveColor, Scale Factor=&MakeVector4_FromFloat(Float=0.25))
       LerpFactor=&FloatFromCurve(FloatCurve=DataInterface() {propertyValues {Curve {keys [<at 0 0 RCIM_Cubic>, {arriveTangent=2.2656, interpMode=RCIM_Cubic, leaveTangent=2.2656, tangentMode=RCTM_User, time=1, value=1}]}}})
      )
      False Color=<rgba 1 1 1 1>
     ))
    ScaleA=true
    ScaleRGB=true
   )
   &DynamicMaterialParameters(
    Index 0 Param 1=&FloatFromCurve(FloatCurve=DataInterface() {propertyValues {Curve {keys [
         <at 0 1 RCIM_Cubic>
         {arriveTangent=-1.7387, arriveTangentWeight=1.9926, interpMode=RCIM_Cubic, leaveTangent=-1.7387, tangentMode=RCTM_User, time=1, value=0.125}
        ]}}})
    Index 0 Param 3=-1
    Index 0 Param 4=&MakeCustomFloatFromBool(
     Bool=$User.UseEmissiveSmoke
     True Float=&FloatFromCurve(FloatCurve=DataInterface() {propertyValues {Curve {keys [{arriveTangent=-2.94, interpMode=RCIM_Cubic, time=0, value=1}, <at 0.75 0 RCIM_Cubic>]}}})
     False Float=0
    )
    Index 1 Param 1=&FloatFromCurve(FloatCurve=DataInterface() {propertyValues {Curve {keys [<at 0 1 RCIM_Cubic>]}}})
    Index 1 Param 2=&FloatFromCurve(FloatCurve=DataInterface() {propertyValues {Curve {keys [{arriveTangent=-2.2405, interpMode=RCIM_Cubic, leaveTangent=-2.2405, tangentMode=RCTM_User, time=0, value=1}, <at 1 0 RCIM_Cubic>]}}})
    Index 1 Param 3=$Emitter.SystemRand01
    Module.Index0Float0Write=true
    Module.Index0Float1Write=true
    Module.Index0Float2Write=true
    Module.Index0Float3Write=true
    Module.Index1Float0Write=true
    Module.Index1Float1Write=true
    Module.Index1Float2Write=true
    Module.Index1Float3Write=true
    Write Parameter Index 1=::NewEnumerator1
   )
   &ScaleRibbonWidth(Ribbon Width Scale=&FloatFromCurve(FloatCurve=DataInterface() {propertyValues {Curve {keys [{tangentMode=RCTM_User, time=0, value=1}, {tangentMode=RCTM_User, time=1, value=0}]}}}))
   &SolveForcesAndVelocity(Clamp Velocity=false, Limit Acceleration=false)
   &ScaleColor(Scale Alpha=$User.Smoke_Alpha, ScaleA=true, ScaleRGB=true) {module=ScaleColor001}
   &ScaleColor(Scale Alpha=$User.Global_Smoke_Opacity_Multiplier, ScaleA=true, ScaleRGB=true) {enabled=false, module=ScaleColor002}
  ]
  rendererClasses [&NiagaraRibbonRendererProperties]
  renderers [{rendererClass=&NiagaraRibbonRendererProperties}]
 }
 Burger_B {
  emitterSpawnScript [/SetVariables(Emitter.SystemRand01:NiagaraFloat=&RandomRangeFloat(Minimum=0, Maximum=1)) {isSetParameters=true, module=SetVariables_#1}]
  emitterUpdateScript [
   &EmitterState(Life Cycle Mode=::NewEnumerator1, Loop Duration=1, UseLoopDelay=false)
   &SpawnRate(SpawnRate=30, Use Spawn Probability=false)
  ]
  particleSpawnScript [
   &InitializeParticle(
    Color=$User.BaseColorAlpha
    Lifetime=0.8
    Ribbon Facing Mode=::NewEnumerator1
    Ribbon Facing Vector:Vector3f=<1 0 0>
    Ribbon Width=&Multiply_Float(A=15, B=$User.RibbonScale)
    Ribbon Width Mode=::NewEnumerator1
    UsePositionOffset=false
    Write Material Random=true
   )
   &SystemLocation(
    MaskSpawnGroup=false
    Offset=&Lerp_Vector(A=<-16.5412 -6.6436 -12.9237>, B=<-19.4543 -6.89 -13.3205>, Alpha=$User.Smoke_Large_Offset)
   ) {module=SystemLocation002}
   &SystemLocation(MaskSpawnGroup=false, Offset=&AddVectorToPosition(Position=$Particles.Position, Vector=<0 0 -10>)) {module=SystemLocation001}
   &AddVelocity(Velocity=<0 0 60>)
  ]
  particleUpdateScript [
   &ParticleState()
   &ScaleColor(
    Scale Alpha=&Multiply_Float(
     A=&FloatFromCurve(
      FloatCurve=DataInterface() {propertyValues {Curve {keys [<at 0 0>, {arriveTangent=-1.7714, interpMode=RCIM_Cubic, leaveTangent=-1.7714, tangentMode=RCTM_User, time=0.5, value=1}, <at 0.8 0>]}}}
      Scale Curve=&FNTimeOfDayScalar(Night Scale=1, Morning Scale=0.3, Evening Scale=0.5)
     )
     B=&FloatFromCurve(FloatCurve=DataInterface() {propertyValues {Curve {keys [<at 0 1>,<at 0.3 1>,<at 0.85 1>,<at 1 0>]}}})
    )
    Scale RGB=&MakeVectorFromLinearColorRGB(Color=&MakeCustomLinearColorFromBool(
      Bool=$User.UseEmissiveSmoke
      True Color=&LerpLinearColors(
       StartColor=$User.EmissiveColor
       EndColor=&MultiplyLinearColor(Linear Color=$User.EmissiveColor, Scale Factor=&MakeVector4_FromFloat(Float=0.25))
       LerpFactor=&FloatFromCurve(FloatCurve=DataInterface() {propertyValues {Curve {keys [<at 0 0 RCIM_Cubic>, {arriveTangent=2.2656, interpMode=RCIM_Cubic, leaveTangent=2.2656, tangentMode=RCTM_User, time=1, value=1}]}}})
      )
      False Color=<rgba 1 1 1 1>
     ))
    ScaleA=true
    ScaleRGB=true
   )
   &DynamicMaterialParameters(
    Index 0 Param 1=&FloatFromCurve(FloatCurve=DataInterface() {propertyValues {Curve {keys [
         <at 0 1 RCIM_Cubic>
         {arriveTangent=-1.7387, arriveTangentWeight=1.9926, interpMode=RCIM_Cubic, leaveTangent=-1.7387, tangentMode=RCTM_User, time=1, value=0.125}
        ]}}})
    Index 0 Param 3=-1
    Index 0 Param 4=&MakeCustomFloatFromBool(
     Bool=$User.UseEmissiveSmoke
     True Float=&FloatFromCurve(FloatCurve=DataInterface() {propertyValues {Curve {keys [{arriveTangent=-2.94, interpMode=RCIM_Cubic, time=0, value=1}, <at 0.75 0 RCIM_Cubic>]}}})
     False Float=0
    )
    Index 1 Param 1=&FloatFromCurve(FloatCurve=DataInterface() {propertyValues {Curve {keys [<at 0 1 RCIM_Cubic>]}}})
    Index 1 Param 2=&FloatFromCurve(FloatCurve=DataInterface() {propertyValues {Curve {keys [{arriveTangent=-2.2405, interpMode=RCIM_Cubic, leaveTangent=-2.2405, tangentMode=RCTM_User, time=0, value=1}, <at 1 0 RCIM_Cubic>]}}})
    Index 1 Param 3=$Emitter.SystemRand01
    Module.Index0Float0Write=true
    Module.Index0Float1Write=true
    Module.Index0Float2Write=true
    Module.Index0Float3Write=true
    Module.Index1Float0Write=true
    Module.Index1Float1Write=true
    Module.Index1Float2Write=true
    Module.Index1Float3Write=true
    Write Parameter Index 1=::NewEnumerator1
   )
   &ScaleRibbonWidth(Ribbon Width Scale=&FloatFromCurve(FloatCurve=DataInterface() {propertyValues {Curve {keys [{tangentMode=RCTM_User, time=0, value=1}, {tangentMode=RCTM_User, time=1, value=0}]}}}))
   &SolveForcesAndVelocity(Clamp Velocity=false, Limit Acceleration=false)
   &ScaleColor(Scale Alpha=$User.Smoke_Alpha_B, ScaleA=true, ScaleRGB=true) {module=ScaleColor001}
   &ScaleColor(Scale Alpha=$User.Global_Smoke_Opacity_Multiplier, ScaleA=true, ScaleRGB=true) {enabled=false, module=ScaleColor002}
  ]
  rendererClasses [&NiagaraRibbonRendererProperties]
  renderers [{rendererClass=&NiagaraRibbonRendererProperties}]
 }
 Burger_C {
  emitterSpawnScript [/SetVariables(Emitter.SystemRand01:NiagaraFloat=&RandomRangeFloat(Minimum=0, Maximum=1)) {isSetParameters=true, module=SetVariables_#1}]
  emitterUpdateScript [
   &EmitterState(Life Cycle Mode=::NewEnumerator1, Loop Duration=1, UseLoopDelay=false)
   &SpawnRate(SpawnRate=30, Use Spawn Probability=false)
  ]
  particleSpawnScript [
   &InitializeParticle(
    Color=$User.BaseColorAlpha
    Lifetime=0.8
    Ribbon Facing Mode=::NewEnumerator1
    Ribbon Facing Vector:Vector3f=<1 0 0>
    Ribbon Width=&Multiply_Float(A=15, B=$User.RibbonScale)
    Ribbon Width Mode=::NewEnumerator1
    UsePositionOffset=false
    Write Material Random=true
   )
   &SystemLocation(
    MaskSpawnGroup=false
    Offset=&Lerp_Vector(
     A=<17.5666 0.6719 -6.7835>
     B=&Lerp_Vector(A=<17.1895 0.2585 -6.6443>, B=<18.475 -4.8222 -7.3953>, Alpha=$User.Smoke_Large_Offset)
     Alpha=$User.Smoke_Large_Offset
    )
   )
   &SystemLocation(MaskSpawnGroup=false, Offset=&AddVectorToPosition(Position=$Particles.Position, Vector=<0 0 -15>)) {module=SystemLocation001}
   &AddVelocity(Velocity=<0 0 60>)
  ]
  particleUpdateScript [
   &ParticleState()
   &ScaleColor(
    Scale Alpha=&Multiply_Float(
     A=&FloatFromCurve(
      FloatCurve=DataInterface() {propertyValues {Curve {keys [<at 0 0>, {arriveTangent=-1.7714, interpMode=RCIM_Cubic, leaveTangent=-1.7714, tangentMode=RCTM_User, time=0.5, value=1}, <at 0.8 0>]}}}
      Scale Curve=&FNTimeOfDayScalar(Night Scale=1, Morning Scale=0.3, Evening Scale=0.5)
     )
     B=&FloatFromCurve(FloatCurve=DataInterface() {propertyValues {Curve {keys [<at 0 1>,<at 0.3 1>,<at 0.85 1>,<at 1 0>]}}})
    )
    Scale RGB=&MakeVectorFromLinearColorRGB(Color=&MakeCustomLinearColorFromBool(
      Bool=$User.UseEmissiveSmoke
      True Color=&LerpLinearColors(
       StartColor=$User.EmissiveColor
       EndColor=&MultiplyLinearColor(Linear Color=$User.EmissiveColor, Scale Factor=&MakeVector4_FromFloat(Float=0.25))
       LerpFactor=&FloatFromCurve(FloatCurve=DataInterface() {propertyValues {Curve {keys [<at 0 0 RCIM_Cubic>, {arriveTangent=2.2656, interpMode=RCIM_Cubic, leaveTangent=2.2656, tangentMode=RCTM_User, time=1, value=1}]}}})
      )
      False Color=<rgba 1 1 1 1>
     ))
    ScaleA=true
    ScaleRGB=true
   )
   &DynamicMaterialParameters(
    Index 0 Param 1=&FloatFromCurve(FloatCurve=DataInterface() {propertyValues {Curve {keys [
         <at 0 1 RCIM_Cubic>
         {arriveTangent=-1.7387, arriveTangentWeight=1.9926, interpMode=RCIM_Cubic, leaveTangent=-1.7387, tangentMode=RCTM_User, time=1, value=0.125}
        ]}}})
    Index 0 Param 3=-1
    Index 0 Param 4=&MakeCustomFloatFromBool(
     Bool=$User.UseEmissiveSmoke
     True Float=&FloatFromCurve(FloatCurve=DataInterface() {propertyValues {Curve {keys [{arriveTangent=-2.94, interpMode=RCIM_Cubic, time=0, value=1}, <at 0.75 0 RCIM_Cubic>]}}})
     False Float=0
    )
    Index 1 Param 1=&FloatFromCurve(FloatCurve=DataInterface() {propertyValues {Curve {keys [<at 0 1 RCIM_Cubic>]}}})
    Index 1 Param 2=&FloatFromCurve(FloatCurve=DataInterface() {propertyValues {Curve {keys [{arriveTangent=-2.2405, interpMode=RCIM_Cubic, leaveTangent=-2.2405, tangentMode=RCTM_User, time=0, value=1}, <at 1 0 RCIM_Cubic>]}}})
    Index 1 Param 3=$Emitter.SystemRand01
    Module.Index0Float0Write=true
    Module.Index0Float1Write=true
    Module.Index0Float2Write=true
    Module.Index0Float3Write=true
    Module.Index1Float0Write=true
    Module.Index1Float1Write=true
    Module.Index1Float2Write=true
    Module.Index1Float3Write=true
    Write Parameter Index 1=::NewEnumerator1
   )
   &ScaleRibbonWidth(Ribbon Width Scale=&FloatFromCurve(FloatCurve=DataInterface() {propertyValues {Curve {keys [{tangentMode=RCTM_User, time=0, value=1}, {tangentMode=RCTM_User, time=1, value=0}]}}}))
   &SolveForcesAndVelocity(Clamp Velocity=false, Limit Acceleration=false)
   &ScaleColor(ScaleA=true, ScaleRGB=true) {module=ScaleColor001}
   &ScaleColor(Scale Alpha=$User.Global_Smoke_Opacity_Multiplier, ScaleA=true, ScaleRGB=true) {enabled=false, module=ScaleColor002}
  ]
  rendererClasses [&NiagaraRibbonRendererProperties]
  renderers [{rendererClass=&NiagaraRibbonRendererProperties}]
 }
}
summary {
 systemName=NS_Emote_CyclopsPrey_Sizzle_Consolidation
 emitters [
  {emitterName=Ambient_Steam, rendererClasses [&NiagaraSpriteRendererProperties]}
  {emitterName=Burger_A, rendererClasses [&NiagaraRibbonRendererProperties]}
  {emitterName=Burger_B, rendererClasses [&NiagaraRibbonRendererProperties]}
  {emitterName=Burger_C, rendererClasses [&NiagaraRibbonRendererProperties]}
 ]
 userVariables [
  User.BaseColorAlpha {type=LinearColor, description="", defaultValue {struct=LinearColor, value=<rgba 1 1 1 0.1>}}
  User.EmissiveColor {type=LinearColor, description="", defaultValue {struct=LinearColor, value=<rgba 1 1 1 0.761>}}
  User.Global_Smoke_Opacity_Multiplier {type=NiagaraFloat, defaultValue=0, description=""}
  User.RibbonScale {type=NiagaraFloat, defaultValue=1.5, description=""}
  User.Smoke_Alpha {type=NiagaraFloat, defaultValue=2, description=""}
  User.Smoke_Alpha_B {type=NiagaraFloat, defaultValue=0, description=""}
  User.Smoke_Large_Offset {type=NiagaraFloat, defaultValue=0, description=""}
  User.Smoke_Position_Offset {type=NiagaraFloat, defaultValue=0, description=""}
  User.UseEmissiveSmoke {type=NiagaraBool}
 ]
}
